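# Niagara effect for the Happy Birthday emote that shows purple cake splatter on a face, which then drips off.
# Two live burst emitters spawn sprite splats and darker mesh chunks in an offset, camera-biased hemisphere.
# Curve-driven gravity with per-particle variation delays their fall, while sprites stretch into thin drips and chunks shrink away.
# An older splat emitter and a skeletal-mesh collision-event emitter, the only user of User.SkeletalMesh, are disabled.
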NiagaraSystem /BRCosmetics/Animation/Game/MainPlayer/Emotes/HappyBirthday/FX/NS_HB_CakeOnFace
dependencies {
 usedDataInterfaces [&NiagaraDataInterfaceCurve, &NiagaraDataInterfaceVector2DCurve]
 usedDynamicInputs []
 usedModules [&SystemState]
}
emitters {
 CakeSplat {
  bEnabled=false
  emitterUpdateScript [&EmitterState()]
  particleSpawnScript [
   /SetVariables(Particles.StableRandom:NiagaraFloat=&UniformRangedFloat(Minimum=0, Maximum=1)) {isSetParameters=true}
   /Niagara/Modules/Spawn/Initialization/InitializeParticle(
    Color=&LerpLinearColors(
     StartColor=<rgba 0.8542 0.125 1 1>
     EndColor=<rgba 0.4576 0.1146 0.5729 1>
     LerpFactor=/Niagara/DynamicInputs/Integer/V2/Round_Float(Float:NiagaraFloat=$Particles.StableRandom)
    )
    Lifetime=&UniformRangedFloat(Minimum=0.9, Maximum=1.3)
    Module.Write Color=true
    Module.Write Lifetime=true
    Module.Write Mass=true
    Module.Write Material Random=true
    Module.Write Position=false
    Module.Write Scale=false
    Module.Write SpriteRotation=true
    Module.Write SpriteSize=true
    Position=&SimulationPosition()
    Sprite Rotation=&UniformRangedFloat(Minimum=0, Maximum=360)
    Sprite Size=&Vector2DFromFloat(Value=&Lerp_Float(A=5, B=22, Alpha=&Power(a=$Particles.StableRandom, b=1.4)))
   )
   /SetVariables(Particles.SplatOffset:Vector3f=<0 -6 0>) {enabled=false, isSetParameters=true}
   &CameraOffset(Camera Offset Amount=2) {enabled=false}
  ]
  particleUpdateScript [
   &ParticleState()
   &GravityForce(Gravity=&MakeVector(
     X=0
     Y=0
     Z=&FloatFromCurve(
      FloatCurve=DataInterface() {propertyValues {Curve {keys [{arriveTangent=-2041.0665, interpMode=RCIM_Cubic, time=0.22, value=-0.01}, <at 1 -400>]}}}
      Scale Curve=&ClampFloat(Float=&Add_Float(A=$Particles.Initial.MaterialRandom, B=0.1))
     )
    ))
   &SolveForcesAndVelocity(Clamp Velocity=false, Limit Acceleration=false)
   &ScaleSpriteSize(Scale Factor=&Vector2DFromCurve(
     Vector2Curve=DataInterface() {propertyValues {
       XCurve {keys [<at 0 1.4 RCIM_Cubic>, {arriveTangent=-2.0041, interpMode=RCIM_Cubic, leaveTangent=-2.0041, tangentMode=RCTM_User, time=1, value=0}]}
       YCurve {keys [<at 0 1 RCIM_Cubic>, {arriveTangent=2.122, interpMode=RCIM_Cubic, leaveTangent=2.122, tangentMode=RCTM_User, time=1, value=2}]}
      }}
     ScaleCurve=<1 1>
    ))
  ]
  rendererClasses [&NiagaraSpriteRendererProperties]
  renderers [{rendererClass=&NiagaraSpriteRendererProperties}]
 }
 CakeSplat001 {
  emitterUpdateScript [
   &EmitterState()
   &SpawnBurst_Instantaneous(Spawn Count=&UniformRangedInt(Minimum=3, Maximum=6), Use Spawn Probability=false)
  ]
  particleSpawnScript [
   &InitializeParticle(
    Color=&LerpLinearColors(
     StartColor=<rgba 0.8542 0.125 1 1>
     EndColor=<rgba 0.4576 0.1146 0.5729 1>
     LerpFactor=&Round_Float(Float:NiagaraFloat=&UniformRangedFloat(Minimum=0.1, Maximum=1.5))
    )
    Lifetime=&UniformRangedFloat(Minimum=1.2, Maximum=1.6)
    Module.Write Color=true
    Module.Write Lifetime=true
    Module.Write Mass=true
    Module.Write Material Random=true
    Module.Write Position=false
    Module.Write Scale=false
    Module.Write SpriteRotation=true
    Module.Write SpriteSize=true
    Position=&SimulationPosition()
    Sprite Rotation=&UniformRangedFloat(Minimum=0, Maximum=360)
    Sprite Size=&Vector2DFromFloat(Value=&Lerp_Float(A=20, B=35, Alpha=&Power(a=$Particles.StableRandom, b=1.4)))
   )
   &SystemLocation(MaskSpawnGroup=false, Offset=&UniformRangedVector(Minimum=<-3.5 -18 35>, Maximum=<-3.5 -12 35>))
   &SphereLocation(
    Hemisphere Y=true
    MaskSpawnGroup=false
    Non Uniform Scale=<0.7 0.7 1>
    Override Seed=false
    Sphere Radius=10
    Surface Only=false
    UseNonUniformScale=true
    UseOffset=false
    UseSphereOrientation=false
   )
   &CameraOffset(Camera Offset Amount=&UniformRangedFloat(Minimum=7, Maximum=10))
  ]
  particleUpdateScript [
   &ParticleState()
   &GravityForce(Gravity=&MakeVector(
     X=0
     Y=0
     Z=&FloatFromCurve(
      FloatCurve=DataInterface() {propertyValues {Curve {keys [{arriveTangent=-2041.0665, interpMode=RCIM_Cubic, time=0.22, value=-0.01}, <at 1 -400>]}}}
      Scale Curve=&ClampFloat(Float=&Add_Float(A=$Particles.Initial.MaterialRandom, B=0.1))
     )
    ))
   &SolveForcesAndVelocity(Clamp Velocity=false, Limit Acceleration=false)
   &ScaleSpriteSize(Scale Factor=&Vector2DFromCurve(
     Vector2Curve=DataInterface() {propertyValues {
       XCurve {keys [<at 0 1.4 RCIM_Cubic>, {arriveTangent=-2.0041, interpMode=RCIM_Cubic, leaveTangent=-2.0041, tangentMode=RCTM_User, time=1, value=0}]}
       YCurve {keys [<at 0 1 RCIM_Cubic>, {arriveTangent=2.122, interpMode=RCIM_Cubic, leaveTangent=2.122, tangentMode=RCTM_User, time=1, value=2}]}
      }}
     ScaleCurve=<1 1>
    ))
  ]
  rendererClasses [&NiagaraSpriteRendererProperties]
  renderers [{rendererClass=&NiagaraSpriteRendererProperties}]
 }
 CakeSplat002 {
  emitterUpdateScript [
   &EmitterState()
   &SpawnBurst_Instantaneous(Spawn Count=&UniformRangedInt(Minimum=3, Maximum=4), Use Spawn Probability=false)
  ]
  particleSpawnScript [
   &InitializeParticle(
    Color=&LerpLinearColors(
     StartColor=<rgba 0.2517 0.0516 0.2917 1>
     EndColor=<rgba 0.1339 0.0434 0.2135 1>
     LerpFactor=&Round_Float(Float:NiagaraFloat=&UniformRangedFloat(Minimum=0.1, Maximum=1.25))
    )
    Lifetime=&UniformRangedFloat(Minimum=1.2, Maximum=1.6)
    Mesh Scale=&Multiply_Vector(
     A=<0.3 1 1>
     B=&VectorFromFloat(Value=&UniformRangedFloat(Minimum=0.25, Maximum=0.6))
    )
    Module.Write Color=true
    Module.Write Lifetime=true
    Module.Write Mass=true
    Module.Write Material Random=true
    Module.Write Position=false
    Module.Write Scale=true
    Module.Write SpriteRotation=false
    Module.Write SpriteSize=false
    Position=&SimulationPosition()
   )
   &SystemLocation(MaskSpawnGroup=false, Offset=&UniformRangedVector(Minimum=<-3.5 -18 35>, Maximum=<-3.5 -12 35>))
   &SphereLocation(
    Hemisphere Y=true
    MaskSpawnGroup=false
    Non Uniform Scale=<0.7 0.7 1>
    Override Seed=false
    Sphere Radius=10
    Surface Only=false
    UseNonUniformScale=true
    UseOffset=false
    UseSphereOrientation=false
   )
   &CameraOffset(Camera Offset Amount=&UniformRangedFloat(Minimum=7, Maximum=10))
   /Niagara/Modules/Spawn/Orientation/V2/InitialMeshOrientation(
    Rotation=&UniformRangedVector(Minimum=<0 0 0>, Maximum=<1 1 1>)
    Use Orientation Vector=true
    Use Rotation Vector=true
   )
  ]
  particleUpdateScript [
   &ParticleState()
   &GravityForce(Gravity=&MakeVector(
     X=0
     Y=0
     Z=&FloatFromCurve(
      FloatCurve=DataInterface() {propertyValues {Curve {keys [{arriveTangent=-2041.0665, interpMode=RCIM_Cubic, time=0.1, value=-0.01}, <at 1 -400>]}}}
      Scale Curve=&ClampFloat(Float=&Add_Float(A=$Particles.Initial.MaterialRandom, B=0.1))
     )
    ))
   &SolveForcesAndVelocity(Clamp Velocity=false, Limit Acceleration=false)
   &ScaleMeshSize(Scale Factor=&VectorFromFloat(Value=&FloatFromCurve(FloatCurve=DataInterface() {propertyValues {Curve {keys [<at 0.425 1 RCIM_Cubic>,<at 1 0>]}}})))
   &DynamicMaterialParameters(
    Index 0 Param 1=5.5
    Index 0 Param 2=3
    Index 0 Param 3=2
    Module.Param0WriteEnabled=true
    Module.Param1WriteEnabled=true
    Module.Param2WriteEnabled=true
    Module.Param3WriteEnabled=true
   )
  ]
  rendererClasses [&NiagaraMeshRendererProperties]
  renderers [{rendererClass=&NiagaraMeshRendererProperties, rendererIndex=0}]
 }
 EventGenerator {
  bEnabled=false
  emitterUpdateScript [
   &EmitterState()
   &SpawnBurst_Instantaneous(Spawn Count=200, Use Spawn Probability=false)
  ]
  particleSpawnScript [
   &InitializeParticle(
    Lifetime=2
    Module.Write Color=true
    Module.Write Lifetime=true
    Module.Write Mass=true
    Module.Write Material Random=true
    Module.Write Position=false
    Module.Write Scale=false
    Module.Write SpriteRotation=false
    Module.Write SpriteSize=true
    Position=&SimulationPosition()
    Sprite Size=&Vector2DFromFloat(Value=2)
   )
   &SkeletalMeshLocation(
    Bone Sampling Mode=::NewEnumerator4
    Inherit Velocity=false
    "Inherit Velocity (Scale)"=0
    MaskSpawnGroup=false
    Sampled Position Offset=<0 8 50>
    Sampled Position Offset Coordinate Space:ENiagaraMeshTransforms=::NewEnumerator3
    Skeletal Mesh=$User.SkeletalMesh
   )
   &AddVelocityInCone(Cone Angle=25, Cone Axis=<0 -1 0>, Module.Use Velocity Falloff On Cone Axis=false, Velocity Strength=400)
   &SphereLocation(
    Hemisphere Z=true
    MaskSpawnGroup=false
    Override Seed=false
    Sphere Origin:NiagaraPosition=/Niagara/DynamicInputs/Helpers/SimulationPosition()
    Sphere Radius=1
    Surface Only=false
    UseNonUniformScale=false
    UseOffset=false
    UseSphereOrientation=false
   ) {enabled=false}
   &AddVelocityFromPoint(Module.Use Falloff Distance=false, Velocity Strength=300) {enabled=false}
  ]
  particleUpdateScript [
   &ParticleState()
   &Collision(
    Age Colliding Particles=false
    Module.Control Rest State Via Surface Normal=false
    Module.EnableMaxCollisionCount=false
    Module.Randomize Collision Normal Vector=false
    Particle Radius Scale=0
    Restitution=0
   )
   /Niagara/Modules/Events/GenerateCollisionEvent(Module.Use Time Between Events=false, Module.Use Velocity Threshold=true)
   &SolveForcesAndVelocity(Clamp Velocity=false, Limit Acceleration=false)
   &KillParticles(Kill Particles=$Particles.HasCollided)
  ]
  rendererClasses [&NiagaraSpriteRendererProperties]
  renderers [{rendererClass=&NiagaraSpriteRendererProperties}]
 }
}
summary {
 systemName=NS_HB_CakeOnFace
 emitters [
  {bEnabled=false, emitterName=EventGenerator, rendererClasses [&NiagaraSpriteRendererProperties]}
  {bEnabled=false, emitterName=CakeSplat, rendererClasses [&NiagaraSpriteRendererProperties]}
  {emitterName=CakeSplat001, rendererClasses [&NiagaraSpriteRendererProperties]}
  {emitterName=CakeSplat002, rendererClasses [&NiagaraMeshRendererProperties]}
 ]
 userVariables [User.SkeletalMesh {
   type=NiagaraDataInterfaceSkeletalMesh
   defaultValue {
    struct=NiagaraExt_VariableValue_DataInterface
    value {dataInterface=&self:NiagaraDataInterfaceSkeletalMesh_1}
   }
  }]
}
