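# Niagara effect for the Happy Birthday emote that renders a candle flame being blown out, with short-lived sparks.
# NE_CandleFlame bursts a ribbon from a user skeletal mesh, using exec index to grade velocity, lifetime, width and curl noise.
# An index comparison gates re-sampling of mesh position each update, keeping the ribbon tied to the mesh.
# NE_CancleSparks adds burst and continuous sprites from a small sphere, thrown along X under randomized gravity.
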NiagaraSystem /BRCosmetics/Animation/Game/MainPlayer/Emotes/HappyBirthday/FX/NS_HB_CandleBlowRibbon
dependencies {usedDataInterfaces [&NiagaraDataInterfaceCurve], usedDynamicInputs [], usedModules [&SystemState]}
emitters {
 NE_CancleSparks {
  emitterUpdateScript [
   &EmitterState()
   &SpawnRate(SpawnRate=90, Use Spawn Probability=false)
   &SpawnBurst_Instantaneous(Spawn Count=&UniformRangedInt(Minimum=15, Maximum=10), Use Spawn Probability=false)
  ]
  particleSpawnScript [
   /Niagara/Modules/Spawn/Initialization/InitializeParticle(
    Color=<rgba 15 10 1 1>
    Lifetime=&UniformRangedFloat(Minimum=0.12, Maximum=0.2)
    Module.Write Color=true
    Module.Write Lifetime=true
    Module.Write Mass=true
    Module.Write Material Random=true
    Module.Write Position=true
    Module.Write Scale=false
    Module.Write SpriteRotation=false
    Module.Write SpriteSize=true
    Position=&SimulationPosition()
    Sprite Size=&Vector2DFromFloat(Value=&UniformRangedFloat(Minimum=0.333, Maximum=1))
   )
   &SphereLocation(
    MaskSpawnGroup=false
    Offset=&UniformRangedVector(Minimum=<0 0 -2>, Maximum=<0 0 -0.5>)
    Override Seed=false
    Sphere Radius=1.5
    Surface Only=false
    UseNonUniformScale=false
    UseOffset=true
    UseSphereOrientation=false
   )
   &AddVelocity(Velocity=&UniformRangedVector(Minimum=<400 0 0>, Maximum=<90 0 0>))
   &AddVelocityFromPoint(Module.Use Falloff Distance=false, Velocity Strength=&UniformRangedFloat(Minimum=15, Maximum=20))
   /SetVariables(Particles.StableGravity:NiagaraFloat=&UniformRangedFloat(Minimum=-450, Maximum=-600)) {isSetParameters=true}
  ]
  particleUpdateScript [
   &ParticleState()
   &GravityForce(Gravity=&MakeVector(X=0, Y=0, Z=$Particles.Initial.StableGravity))
   &SolveForcesAndVelocity(Clamp Velocity=false, Limit Acceleration=false)
   &ScaleColor(
    Scale Alpha=&FloatFromCurve(FloatCurve=DataInterface() {propertyValues {Curve {keys [<at 0.5 1>,<at 1 0>]}}})
    ScaleA=true
    ScaleRGB=true
   )
  ]
  rendererClasses [&NiagaraSpriteRendererProperties]
  renderers [{rendererClass=&NiagaraSpriteRendererProperties}]
 }
 NE_CandleFlame {
  emitterUpdateScript [
   &EmitterState()
   &SpawnBurst_Instantaneous(Spawn Count=10, Use Spawn Probability=false)
  ]
  particleSpawnScript [
   &InitializeRibbon(
    Color=<rgba 10 1.75 0.2 0.3>
    Lifetime=&Add_Float(A=0.1, B=&Divide_Float(A=&ReturnNormalizedExecIndex(), B=1.8))
    Module.Write Color=true
    Module.Write Lifetime=true
    Module.Write Mass=true
    Module.Write Material Random=true
    Module.Write Position=true
    Module.Write RibbonTwist=false
    Module.Write RibbonWidth=true
    Ribbon Width=2.222
   )
   &SkeletalMeshLocation(
    Bone Sampling Mode=::NewEnumerator4
    Inherit Velocity=true
    MaskSpawnGroup=false
    Sampled Position Offset=<-1.5 -0.9 0>
    Sampled Position Offset Coordinate Space:ENiagaraMeshTransforms=::NewEnumerator3
    Skeletal Mesh=$User.NewNiagaraDataInterfaceSkeletalMesh
    VelocitySpeedLimit=false
   )
   /SetVariables(Particles.RibbonLinkOrder:NiagaraFloat=&ReturnNormalizedExecIndex()) {isSetParameters=true}
   /SetVariables(Particles.RibbonID=/Niagara/DynamicInputs/Meshes/MakeRibbonIDFromSpawnGroup()) {isSetParameters=true}
   /SetVariables(Particles.StableExecIndex:NiagaraInt32=&ReturnExecIndex()) {isSetParameters=true}
   &AddVelocity(Velocity=&Multiply_Vector(
     A=<400 0 0>
     B=&MakeVector(X=&OneMinusFloat(Float=&ReturnNormalizedExecIndex()), Y=0, Z=0)
    ))
  ]
  particleUpdateScript [
   &ParticleState()
   &SkeletalMeshLocation(
    Apply To Particle Position:NiagaraBool=&SetBoolByIntComparison(A=9, B=$Particles.Initial.StableExecIndex, Comparison Type=::NewEnumerator4)
    Bone Sampling Mode=::NewEnumerator4
    Inherit Velocity=false
    Interpolate Mesh Sampling=false
    MaskSpawnGroup=false
    Sampled Position Offset Coordinate Space:ENiagaraMeshTransforms=::NewEnumerator3
    Skeletal Mesh=$User.NewNiagaraDataInterfaceSkeletalMesh
   ) {module=SkeletalMeshLocation001}
   /Niagara/Modules/Update/Forces/CurlNoiseForce(
    Noise Frequency=0.5
    Noise Strength=&Multiply_Float(A=180, B=$Particles.Initial.RibbonLinkOrder)
    Pan Noise=true
    Random Seed:NiagaraInt32=&UniformRangedInt(Minimum=0, Maximum=50)
   )
   &Drag(Drag=5, Use Linear Drag=true, Use Rotational Drag=false)
   &SolveForcesAndVelocity(Clamp Velocity=false, Limit Acceleration=false)
   &ScaleColor(
    Scale Alpha=&FloatFromCurve(FloatCurve=DataInterface() {propertyValues {Curve {keys [<at 0 1 RCIM_Cubic>, {arriveTangent=-1, interpMode=RCIM_Cubic, leaveTangent=-1, time=0.2, value=1}, <at 1 0 RCIM_Cubic>]}}})
    ScaleA=true
    ScaleRGB=true
   )
   &ScaleRibbonWidth(Ribbon Width Scale=&Multiply_Float(A=2.5, B=&ReturnNormalizedExecIndex()))
  ]
  rendererClasses [&NiagaraSpriteRendererProperties, &NiagaraRibbonRendererProperties]
  renderers [
   {rendererClass=&NiagaraSpriteRendererProperties}
   {rendererClass=&NiagaraRibbonRendererProperties, rendererIndex=1}
  ]
 }
}
summary {
 systemName=NS_HB_CandleBlowRibbon
 emitters [
  {
   emitterName=NE_CandleFlame
   rendererClasses [&NiagaraSpriteRendererProperties, &NiagaraRibbonRendererProperties]
  }
  {emitterName=NE_CancleSparks, rendererClasses [&NiagaraSpriteRendererProperties]}
 ]
 userVariables [User.NewNiagaraDataInterfaceSkeletalMesh {
   type=NiagaraDataInterfaceSkeletalMesh
   defaultValue {
    struct=NiagaraExt_VariableValue_DataInterface
    value {dataInterface=&self:NiagaraDataInterfaceSkeletalMesh_0}
   }
  }]
}
